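# Animation Blueprint for the Toxic Storm backpack that nudges its pack bone to fit the wearer's pose and body.
# Chained additive bone transforms apply a base offset, a parachuting offset weighted by Parashuting, and a bool-switched offset.
# Each update infers Parashuting from the owner mesh's spine_05-to-spine_01 yaw delta, remapped and clamped, rather than reading game state.
# On customization load, Character Prime flags whether a body part uses the F_MED_Prime mesh.
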AnimBlueprint /BRCosmetics/Accessories/FORT_Backpacks/Backpack_M_MED_ToxicStorm/Meshes/M_MED_ToxicStorm_Pack_AnimBP
target_skeleton=&/M_MED_ToxicStorm_Pack_Skeleton
graphs [
 AnimGraph {nodes [
   n1 = "Transform (Modify) Bone - Bone: pack"(ComponentPose=Local To Component(LocalPose=Input Pose()), Translation="0,-0.5,4", Rotation="0,5,0") {
    rotationMode=BMM_Additive
    scaleMode=BMM_Additive
    translationMode=BMM_Additive
    alphaBoolBlend {blendOption=Linear}
    boneToModify {boneName=pack}
    rotation {yaw=5}
    translation {y=-0.5, z=4}
   }
   n2 = "Transform (Modify) Bone - Bone: pack"(ComponentPose=@n1, Alpha=Get Parashuting() -> (Parashuting:double), Translation="0,-2.5,8", Rotation="5,0,0") {
    rotationMode=BMM_Additive
    scaleMode=BMM_Additive
    translationMode=BMM_Additive
    alphaBoolBlend {blendOption=Linear}
    boneToModify {boneName=pack}
    rotation {pitch=5}
    translation {y=-2.5, z=8}
   }
   n3 = "Transform (Modify) Bone - Bone: pack"!(ComponentPose=@n2, bAlphaBoolEnabled:bool, Translation="0,2,0", Rotation="0,0,0", Scale) -> (Pose) {
    alphaInputType=Bool
    rotationMode=BMM_Additive
    scaleMode=BMM_Additive
    translationMode=BMM_Additive
    alphaBoolBlend {blendOption=Linear}
    boneToModify {boneName=pack}
    translation {y=2}
   }
   Output Pose(Result=Component To Local(ComponentPose=@n3)) {layerGroup=None}
   Get Character Prime() -> (Character Prime:bool)
  ]}
 EventGraph {nodes [
   Event BlueprintUpdateAnimation()
   n1 = GetPartSkeletalMeshComponent()
   > Is Valid(InputObject=@n1)
   > Is Valid(InputObject=GetAttachParent!(self:Scene Component Object Reference=@n1) -> (ReturnValue:Scene Component Object Reference))
   n2 = GetDeltaTransformFromRefPose!(self=GetOwnerSkeletalMeshComponent(), BoneName=spine_05, BaseName=spine_01) -> (ReturnValue_Location:Vector, ReturnValue_Rotation:Rotator, ReturnValue_Scale:Vector)
   > Set Parashuting(Parashuting:double=MapRangeClamped(Value=BreakRotator(InRot=@n2.ReturnValue_Rotation).Yaw, InRangeA=-25, InRangeB=-65)) -> (Output_Get:double)
   n3 = Event OnCharacterCustomizationLoaded()
   n4 = Get Array Item with EFort Custom Part Type(Array:Array of Custom Character Part Object References=@n3.CustomParts) -> (Output:Custom Character Part Object Reference)
   > Set Character Prime(Character Prime:bool="Equal (Object)"(
     A=GetSkeletalMesh!(self:Custom Character Part Object Reference=@n4) -> (ReturnValue:Skeletal Mesh Object Reference)
     B=/DefaultCosmeticsAthena/Characters/Meshes/F_MED_Prime
    )) -> (Output_Get:bool)
  ]}
]
variables [
 Character Prime {type=boolean}
 Parashuting {type=number}
]
